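# Third-person fire sound for the exotic ShocknSlow Shockwave Launcher, a stereo MetaSound preset of the shared MSS_Base_Weapon_Fire_3p_Complete weapon graph.
# It supplies close, medium and distant wave sets, with distant using generic Shockwave Launcher recordings and suppressor sets left empty.
# Inputs configure ranged-weapon distance curves, layer mixing, and random pitch, volume and EQ for the parent graph.
# Source settings add projectile weapon attenuation, gunfire concurrency and third-person volume modulation.
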MetaSoundSource /TidalCraneExotics/Gameplay/ShockwaveLauncher/Audio/MSS_Exotic_Launcher_ShocknSlow_Fire_3P
preset_parent=/FortMetaSoundWeapon/Items/Metasounds/MSS_Base_Weapon_Fire_3p_Complete
inputs [
 bEnableDistanceEQ {type=Bool}
 bHasUnlimitedAmmo {type=Bool}
 bIsAlwaysSuppressed {type=Bool}
 BurstRate {type=Time}
 bUseSingleVoice {type=Bool}
 Character.Affiliation {type=Enum:FortAudioAffiliation}
 Character.Health {type=Float}
 Character.UniqueID {type=Int32}
 FireBurst {type=Trigger}
 FireRateMultiplier {type=Float}
 FireSingleShot {type=Trigger}
 MainFire.DistanceCurves {type=WaveTableBankAsset}
 MainFire.SoundsClose {type=WaveAsset:Array}
 MainFire.SoundsDistant {type=WaveAsset:Array}
 MainFire.SoundsMedium {type=WaveAsset:Array}
 MasterVolume {type=Float}
 Mixing.FalloffDistance {type=Float}
 Mixing.Friendly.FriendlyFireWeaponVolume {type=Float}
 Mixing.Suppressor.BaseWeaponVolume {type=Float}
 Mixing.Suppressor.SuppressorVolume {type=Float}
 Mixing.Volume.Close {type=Float}
 Mixing.Volume.Distant {type=Float}
 Mixing.Volume.Medium {type=Float}
 Random.EQ.CutoffFrequency.Max {type=Float}
 Random.EQ.CutoffFrequency.Min {type=Float}
 Random.EQ.Gain.Max {type=Float}
 Random.EQ.Gain.Min {type=Float}
 Random.Pitch.Max {type=Float}
 Random.Pitch.Min {type=Float}
 Random.Volume.Max {type=Float}
 Random.Volume.Min {type=Float}
 RetriggerDelayTime {type=Time}
 RoundsPerBurst {type=Int32}
 StartFiring {type=Trigger}
 StopBurst {type=Trigger}
 StopFiring {type=Trigger}
 Suppressor.EnvelopeDecay {type=Time}
 Suppressor.Sounds.Close {type=WaveAsset:Array}
 Suppressor.Sounds.Medium {type=WaveAsset:Array}
 Suppressor.Volume.Close {type=Float}
 Suppressor.Volume.Medium {type=Float}
 Targeting.FilterBandwidth {type=Float}
 Targeting.FilterFrequency {type=Float}
 Targeting.FilterGain {type=Float}
 UE.Attenuation.Distance {type=Float}
 UE.Source.OnPlay {type=Trigger}
 UE.Source.Orientation.Azimuth {type=Float}
 UE.Source.Orientation.Elevation {type=Float}
 UE.Spatialization.Azimuth {type=Float}
 UE.Spatialization.Elevation {type=Float}
 Weapon.AmmoRemaining {type=Int32}
 Weapon.BurstFireRate {type=Float}
 Weapon.ClipSize {type=Int32}
 Weapon.FireRate {type=Float}
 Weapon.FireStartTime {type=Float}
 Weapon.IsAimingDownSights {type=Bool}
 Weapon.IsSuppressorEnabled {type=Bool}
]
interfaces [
 Fort.Character:1.0
 Fort.Weapon:1.0
 UE.Attenuation:1.0
 UE.OutputFormat.Stereo:1.0
 UE.Source.OneShot:1.0
 UE.Source.Orientation:1.0
 UE.Source:1.1
 UE.Spatialization:1.0
]
outputs [
 bIsAlmostFinished {type=Bool}
 "UE.OutputFormat.Stereo.Audio:0" {type=Audio}
 "UE.OutputFormat.Stereo.Audio:1" {type=Audio}
 UE.Source.OneShot.OnFinished {type=Trigger}
]
pages [Default {input_defaults {
   BurstRate=0.4
   Character.Affiliation=0
   Character.Health=1
   Character.UniqueID=0
   FireBurst=null
   FireRateMultiplier=1
   FireSingleShot=null
   MainFire.DistanceCurves=&WeaponRanged_DistanceCurves
   MasterVolume=1
   Mixing.FalloffDistance=27500
   Mixing.Friendly.FriendlyFireWeaponVolume=0.7
   Mixing.Suppressor.BaseWeaponVolume=0.75
   Mixing.Suppressor.SuppressorVolume=1
   Mixing.Volume.Close=0.85
   Mixing.Volume.Distant=0.35
   Mixing.Volume.Medium=0.45
   Random.EQ.CutoffFrequency.Max=20000
   Random.EQ.CutoffFrequency.Min=699.9999
   Random.EQ.Gain.Max=0
   Random.EQ.Gain.Min=-3
   Random.Pitch.Max=1.05
   Random.Pitch.Min=0.95
   Random.Volume.Max=1.15
   Random.Volume.Min=0.85
   RetriggerDelayTime=2
   RoundsPerBurst=2
   StartFiring=null
   StopBurst=null
   StopFiring=null
   Suppressor.EnvelopeDecay=0.195
   Suppressor.Volume.Close=0.5
   Suppressor.Volume.Medium=0.3
   Targeting.FilterBandwidth=4
   Targeting.FilterFrequency=300
   Targeting.FilterGain=-8
   UE.Attenuation.Distance=22000
   UE.Source.OnPlay=null
   UE.Source.Orientation.Azimuth=0
   UE.Source.Orientation.Elevation=0
   UE.Spatialization.Azimuth=0
   UE.Spatialization.Elevation=-2.88
   Weapon.AmmoRemaining=30
   Weapon.BurstFireRate=0.1
   Weapon.ClipSize=30
   Weapon.FireRate=0.5
   Weapon.FireStartTime=0
   Weapon.IsAimingDownSights=null
   Weapon.IsSuppressorEnabled=null
   bEnableDistanceEQ="<MetaSoundBuilderResult.SUCCEEDED: 0>"
   bHasUnlimitedAmmo=null
   bIsAlwaysSuppressed=null
   bUseSingleVoice="<MetaSoundBuilderResult.SUCCEEDED: 0>"
   MainFire.SoundsClose [
    /TidalCraneExotics/Gameplay/ShockwaveLauncher/Audio/SoundWavs/Exotic_Shockwave_ShocknSlow_Fire_3P_Close_01
    /TidalCraneExotics/Gameplay/ShockwaveLauncher/Audio/SoundWavs/Exotic_Shockwave_ShocknSlow_Fire_3P_Close_02
    /TidalCraneExotics/Gameplay/ShockwaveLauncher/Audio/SoundWavs/Exotic_Shockwave_ShocknSlow_Fire_3P_Close_03
   ]
   MainFire.SoundsDistant [
    /TidalCraneExotics/Gameplay/ShockwaveLauncher/Audio/SoundWavs/Shockwave_Launcher_Fire_Distant_01
    /TidalCraneExotics/Gameplay/ShockwaveLauncher/Audio/SoundWavs/Shockwave_Launcher_Fire_Distant_02
    /TidalCraneExotics/Gameplay/ShockwaveLauncher/Audio/SoundWavs/Shockwave_Launcher_Fire_Distant_03
   ]
   MainFire.SoundsMedium [
    /TidalCraneExotics/Gameplay/ShockwaveLauncher/Audio/SoundWavs/Exotic_Shockwave_ShocknSlow_Fire_3P_Med_01
    /TidalCraneExotics/Gameplay/ShockwaveLauncher/Audio/SoundWavs/Exotic_Shockwave_ShocknSlow_Fire_3P_Med_02
    /TidalCraneExotics/Gameplay/ShockwaveLauncher/Audio/SoundWavs/Exotic_Shockwave_ShocknSlow_Fire_3P_Med_03
   ]
   Suppressor.Sounds.Close []
   Suppressor.Sounds.Medium []
  }}]
properties {
 assetImportData=&self:AssetImportData
 attenuationSettings=&Fort_Weapon_Projectile_NonAtt_Athena_Att
 maxDistance=27500
 numChannels=2
 outputFormat=Stereo
 qualitySetting=Fort_Weapons
 sampleRate=48000
 soundClassObject=&Fort_WeaponRanged_Fire_3P_PSM
 concurrencyOverrides {bLimitToOwner=true, maxCount=2, resolutionRule=StopOldest}
 concurrencySet [&Weapon_Gunfire_Local_Conc, &Weapon_Gunfire_Global_Conc]
 modulationSettings {
  volumeRouting=Union
  volumeModulationDestination {bEnableModulation=true, modulators [/FortMetaSoundWeapon/Items/Modulation/PP_Weapon_Fire_3p_Vol]}
 }
}
