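# One-shot stereo MetaSound Source for the landing sound of the exotic Shockwave Launcher's ShocknSlow projectile.
# On play, a shuffle picks one of three landing waves and triggers the wave player with a random pitch shift of base plus or minus spread, both exposed as inputs.
# The player's finish ends the source, which uses grenade bounce attenuation and the ranged-weapon projectile effects sound class.
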MetaSoundSource /TidalCraneExotics/Gameplay/ShockwaveLauncher/Audio/MSS_Exotic_Launcher_ShocknSlow_Projectile_Land
inputs [
 Random.Pitch.Base {type=Float}
 Random.Pitch.Spread {type=Float}
 Sounds {type=WaveAsset:Array}
 UE.Source.OnPlay {type=Trigger}
]
interfaces [UE.OutputFormat.Stereo:1.0, UE.Source.OneShot:1.0, UE.Source:1.1]
outputs [
 "UE.OutputFormat.Stereo.Audio:0" {type=Audio}
 "UE.OutputFormat.Stereo.Audio:1" {type=Audio}
 UE.Source.OneShot.OnFinished {type=Trigger}
]
pages [Default {
  input_defaults {
   Random.Pitch.Base=-0.5
   Random.Pitch.Spread=0.5
   UE.Source.OnPlay=null
   Sounds [
    /TidalCraneExotics/Gameplay/ShockwaveLauncher/Audio/SoundWavs/Exotic_Shockwave_ShocknSlow_Land_01
    /TidalCraneExotics/Gameplay/ShockwaveLauncher/Audio/SoundWavs/Exotic_Shockwave_ShocknSlow_Land_02
    /TidalCraneExotics/Gameplay/ShockwaveLauncher/Audio/SoundWavs/Exotic_Shockwave_ShocknSlow_Land_03
   ]
  }
  nodes [
   n1 = Array.Shuffle:WaveAsset:Array(In Array=^Sounds, Next=^UE.Source.OnPlay) {also {defaults {Enable Shared State="<MetaSoundBuilderResult.SUCCEEDED: 0>"}}}
   WavePlayer_0 = UE.Wave Player:Stereo(
    Play=@n1.On Next
    Wave Asset=@n1.Value
    Pitch Shift=UE.RandomFloat(
     Next=^UE.Source.OnPlay
     Max=UE.Add:Float(PrimaryOperand=^Random.Pitch.Base, AdditionalOperands=^Random.Pitch.Spread).Out
     Min=UE.Subtract:Float(PrimaryOperand=^Random.Pitch.Base, AdditionalOperands=^Random.Pitch.Spread).Out
    ).Value
   ) {also {version [1,1]}}
  ]
  links [
   {dst="@UE.Source.OneShot.OnFinished", src=WavePlayer_0.On Finished}
   {dst="@UE.OutputFormat.Stereo.Audio:0", src=WavePlayer_0.Out Left}
   {dst="@UE.OutputFormat.Stereo.Audio:1", src=WavePlayer_0.Out Right}
  ]
 }]
properties {
 assetImportData=&self:AssetImportData
 attenuationSettings=/Game/Sounds/Attenuation/Fort_Grenade_Bounce_Att
 maxDistance=5500
 numChannels=2
 outputFormat=Stereo
 sampleRate=48000
 soundClassObject=&Fort_WeaponRanged_Projectile_Effects
 volume=1.5
}
