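# Sound cue for the aim-down-sights-in foley of the Shock 'n Slow exotic shockwave launcher.
# A single wave player feeds a Modulator node at the root, left at its default settings.
# It plays quietly in the weapon foley sound class, reusing the weapon reload attenuation for falloff.
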SoundCue /TidalCraneExotics/Gameplay/ShockwaveLauncher/Audio/SC_Exotic_Launcher_ShocknSlow_ADS_IN
root=Modulator_0
nodes [
 {id=Modulator_0, children [WavePlayer_1]}
 {
  id=WavePlayer_1
  properties {soundWaveAssetPtr=/TidalCraneExotics/Gameplay/ShockwaveLauncher/Audio/SoundWavs/Exotic_Shockwave_ShocknSlow_ADS_In}
 }
]
properties {
 attenuationSettings=&Fort_Weapon_Reload_Att
 maxDistance=2250
 soundClassObject=&Fort_WeaponFoley
 volumeMultiplier=0.15
}
