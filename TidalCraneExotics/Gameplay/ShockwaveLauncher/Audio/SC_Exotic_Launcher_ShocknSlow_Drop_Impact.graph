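# Sound cue for the drop impact of the exotic Shockwave Launcher's ShocknSlow variant.
# A random node picks one of three equally weighted landing waves, and a modulator adds slight pitch variation.
# It plays at reduced volume in the ranged projectile effects sound class with weapon-drop attenuation.
# Attenuation overrides also specify HRTF spatialization and distance-based low-pass filtering.
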SoundCue /TidalCraneExotics/Gameplay/ShockwaveLauncher/Audio/SC_Exotic_Launcher_ShocknSlow_Drop_Impact
root=Modulator_0
nodes [
 {id=Modulator_0, children [Random_2], properties {pitchMax=1.1, pitchMin=0.9}}
 {id=Random_2, children [WavePlayer_2,WavePlayer_3,WavePlayer_4], properties {weights [1,1,1]}}
 {
  id=WavePlayer_2
  properties {soundWaveAssetPtr=/TidalCraneExotics/Gameplay/ShockwaveLauncher/Audio/SoundWavs/Exotic_Shockwave_ShocknSlow_Drop_Land_01}
 }
 {
  id=WavePlayer_3
  properties {soundWaveAssetPtr=/TidalCraneExotics/Gameplay/ShockwaveLauncher/Audio/SoundWavs/Exotic_Shockwave_ShocknSlow_Drop_Land_02}
 }
 {
  id=WavePlayer_4
  properties {soundWaveAssetPtr=/TidalCraneExotics/Gameplay/ShockwaveLauncher/Audio/SoundWavs/Exotic_Shockwave_ShocknSlow_Drop_Land_03}
 }
]
properties {
 attenuationSettings=&Fort_Weapon_Drop_Att
 maxDistance=3000
 soundClassObject=&Fort_WeaponRanged_Projectile_Effects
 volumeMultiplier=0.6
 attenuationOverrides {
  bAttenuateWithLPF=true
  falloffDistance=2500
  lPFFrequencyAtMax=2000
  lPFRadiusMax=3000
  lPFRadiusMin=500
  spatializationAlgorithm=SPATIALIZATION_HRTF
  attenuationShapeExtents {x=500}
 }
}
