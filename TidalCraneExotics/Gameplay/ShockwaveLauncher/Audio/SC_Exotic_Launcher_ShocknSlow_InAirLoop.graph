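# Sound cue for the looping in-air sound of the Shock 'n Slow exotic shockwave launcher's projectile.
# A looping wave player feeds a Modulator, then a Doppler node at reduced intensity, so pitch responds to relative motion.
# It plays at low volume in the explosives sound class, borrows the frag grenade in-air attenuation asset, and overrides concurrency to prevent new instances at the limit.
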SoundCue /TidalCraneExotics/Gameplay/ShockwaveLauncher/Audio/SC_Exotic_Launcher_ShocknSlow_InAirLoop
root=Doppler_0
nodes [
 {id=Doppler_0, children [Modulator_0], properties {dopplerIntensity=0.5}}
 {id=Modulator_0, children [WavePlayer_0]}
 {
  id=WavePlayer_0
  properties {bLooping=true, soundWaveAssetPtr=/TidalCraneExotics/Gameplay/ShockwaveLauncher/Audio/SoundWavs/Shockwave_Launcher_InAirLoop_01}
 }
]
properties {
 attenuationSettings=/Game/Sounds/Fort_Weapons/Grenades/Frag/Grenade_Frag_InAir_Att
 bOverrideConcurrency=true
 maxDistance=6300
 soundClassObject=/Game/Sounds/FortSoundClassesAndModes/Fort_Explosives
 volumeMultiplier=0.3
 attenuationOverrides {distanceAlgorithm=Logarithmic, falloffDistance=1200, spatializationAlgorithm=SPATIALIZATION_HRTF, attenuationShapeExtents {x=500}}
 concurrencyOverrides {resolutionRule=PreventNew}
}
