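# Sound cue for the out-of-ammo sound of the exotic Shock 'n Slow (shockwave) launcher.
# A Random node picks one of four stock launcher out-of-ammo waves with equal weight.
# A Modulator raises the pitch by a fixed amount and slightly varies the volume.
# The cue overrides concurrency to a small cap that stops the oldest instance, and uses the third-person ranged weapon fire sound class.
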SoundCue /TidalCraneExotics/Gameplay/ShockwaveLauncher/Audio/SC_Exotic_Launcher_ShocknSlow_OutOfAmmo
root=Modulator_1
nodes [
 {id=Modulator_1, children [Random_0], properties {pitchMax=1.2, pitchMin=1.2, volumeMax=0.7, volumeMin=0.6}}
 {id=Random_0, children [WavePlayer_1,WavePlayer_2,WavePlayer_3,WavePlayer_4], properties {weights [1,1,1,1]}}
 {id=WavePlayer_1, properties {soundWaveAssetPtr=/Game/Sounds/Fort_Weapons/OutofAmmo/Launcher_OutofAmmo_01}}
 {id=WavePlayer_2, properties {soundWaveAssetPtr=/Game/Sounds/Fort_Weapons/OutofAmmo/Launcher_OutofAmmo_02}}
 {id=WavePlayer_3, properties {soundWaveAssetPtr=/Game/Sounds/Fort_Weapons/OutofAmmo/Launcher_OutofAmmo_03}}
 {id=WavePlayer_4, properties {soundWaveAssetPtr=/Game/Sounds/Fort_Weapons/OutofAmmo/Launcher_OutofAmmo_04}}
]
properties {
 bOverrideConcurrency=true
 maxDistance=2097152
 soundClassObject=&Fort_WeaponRanged_Fire_3P_PSM
 volumeMultiplier=0.55
 attenuationOverrides {falloffDistance=1600}
 concurrencyOverrides {maxCount=2, resolutionRule=StopOldest}
}
