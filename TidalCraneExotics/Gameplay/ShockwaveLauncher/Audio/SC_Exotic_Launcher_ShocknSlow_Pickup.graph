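# Pickup sound cue for the Shock 'n Slow exotic shockwave launcher in TidalCraneExotics.
# A Random node picks one of three equally weighted pickup waves and feeds a default Modulator at the root for variation.
# Playback is reduced in volume, spatialized by the shared weapon pickup attenuation, routed to the loot collect sound class and limited by the shared pickup concurrency set.
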SoundCue /TidalCraneExotics/Gameplay/ShockwaveLauncher/Audio/SC_Exotic_Launcher_ShocknSlow_Pickup
root=Modulator_0
nodes [
 {id=Modulator_0, children [Random_0]}
 {id=Random_0, children [WavePlayer_3,WavePlayer_4,WavePlayer_5], properties {weights [1,1,1]}}
 {
  id=WavePlayer_3
  properties {soundWaveAssetPtr=/TidalCraneExotics/Gameplay/ShockwaveLauncher/Audio/SoundWavs/Exotic_Shockwave_ShocknSlow_Pickup_01}
 }
 {
  id=WavePlayer_4
  properties {soundWaveAssetPtr=/TidalCraneExotics/Gameplay/ShockwaveLauncher/Audio/SoundWavs/Exotic_Shockwave_ShocknSlow_Pickup_02}
 }
 {
  id=WavePlayer_5
  properties {soundWaveAssetPtr=/TidalCraneExotics/Gameplay/ShockwaveLauncher/Audio/SoundWavs/Exotic_Shockwave_ShocknSlow_Pickup_03}
 }
]
properties {
 attenuationSettings=&Fort_Weapon_Pickup_Att
 maxDistance=3500
 soundClassObject=&Fort_Loot_Collect
 volumeMultiplier=0.3
 concurrencyOverrides {maxCount=1, resolutionRule=PreventNew}
 concurrencySet [&PickupMaximum_Oldest]
}
