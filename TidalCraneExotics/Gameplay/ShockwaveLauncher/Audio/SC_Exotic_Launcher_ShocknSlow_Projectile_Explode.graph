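# Sound cue for the exploding projectile of the exotic ShocknSlow shockwave launcher, routed to the weapons sound class.
# A top mixer layers two branches with separate close and medium grenade explosion attenuations.
# The close branch mixes a random ShocknSlow explosion wave with a quieter random standard launcher close explosion; the medium branch plays a random distant explosion with slight pitch and volume modulation.
# A concurrency override stops the oldest instance beyond a small limit.
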SoundCue /TidalCraneExotics/Gameplay/ShockwaveLauncher/Audio/SC_Exotic_Launcher_ShocknSlow_Projectile_Explode
root=Mixer_2
nodes [
 {
  id=Attenuation_0
  children [Modulator_13]
  properties {attenuationSettings=/Game/Sounds/Attenuation/Fort_Explosion_Medium_Athena_Grenade_Att}
 }
 {id=Attenuation_1, children [Mixer_0], properties {attenuationSettings=/Game/Sounds/Attenuation/Fort_Explosion_Close_Athena_Grenade_Att}}
 {id=Mixer_0, children [Random_0,Random_4], properties {inputVolume [1,0.3]}}
 {id=Mixer_2, children [Attenuation_1, Attenuation_0], properties {inputVolume [1,1]}}
 {id=Modulator_13, children [Random_5], properties {pitchMax=0.95, pitchMin=0.85, volumeMax=1.1, volumeMin=1}}
 {id=Random_0, children [WavePlayer_0,WavePlayer_1,WavePlayer_2], properties {weights [1,1,1]}}
 {id=Random_4, children [WavePlayer_12, WavePlayer_13, WavePlayer_14], properties {weights [1,1,1]}}
 {id=Random_5, children [WavePlayer_15, WavePlayer_16, WavePlayer_17], properties {weights [1,1,1]}}
 {
  id=WavePlayer_0
  properties {soundWaveAssetPtr=/TidalCraneExotics/Gameplay/ShockwaveLauncher/Audio/SoundWavs/Exotic_Shockwave_ShocknSlow_Explode_01}
 }
 {
  id=WavePlayer_1
  properties {soundWaveAssetPtr=/TidalCraneExotics/Gameplay/ShockwaveLauncher/Audio/SoundWavs/Exotic_Shockwave_ShocknSlow_Explode_02}
 }
 {
  id=WavePlayer_12
  properties {soundWaveAssetPtr=/TidalCraneExotics/Gameplay/ShockwaveLauncher/Audio/SoundWavs/Shockwave_Launcher_Explode_Close_01}
 }
 {
  id=WavePlayer_13
  properties {soundWaveAssetPtr=/TidalCraneExotics/Gameplay/ShockwaveLauncher/Audio/SoundWavs/Shockwave_Launcher_Explode_Close_02}
 }
 {
  id=WavePlayer_14
  properties {soundWaveAssetPtr=/TidalCraneExotics/Gameplay/ShockwaveLauncher/Audio/SoundWavs/Shockwave_Launcher_Explode_Close_03}
 }
 {
  id=WavePlayer_15
  properties {soundWaveAssetPtr=/TidalCraneExotics/Gameplay/ShockwaveLauncher/Audio/SoundWavs/Shockwave_Launcher_Explode_Distant_01}
 }
 {
  id=WavePlayer_16
  properties {soundWaveAssetPtr=/TidalCraneExotics/Gameplay/ShockwaveLauncher/Audio/SoundWavs/Shockwave_Launcher_Explode_Distant_02}
 }
 {
  id=WavePlayer_17
  properties {soundWaveAssetPtr=/TidalCraneExotics/Gameplay/ShockwaveLauncher/Audio/SoundWavs/Shockwave_Launcher_Explode_Distant_03}
 }
 {
  id=WavePlayer_2
  properties {soundWaveAssetPtr=/TidalCraneExotics/Gameplay/ShockwaveLauncher/Audio/SoundWavs/Exotic_Shockwave_ShocknSlow_Explode_03}
 }
]
properties {
 bOverrideConcurrency=true
 maxDistance=9625
 soundClassObject=/Game/Sounds/FortSoundClassesAndModes/Fort_Weapons
 volumeMultiplier=1
 attenuationOverrides {distanceAlgorithm=NaturalSound, falloffDistance=3800, lPFRadiusMax=3000, lPFRadiusMin=2000, attenuationShapeExtents {x=2200}}
 concurrencyOverrides {maxCount=2, resolutionRule=StopOldest}
}
