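# SoundCue for the Shock'n Slow exotic shockwave launcher's projectile exploding in water, on the weapons sound class.
# A root mixer blends close and medium grenade attenuation layers, each choosing random wave variants.
# Close mixes half-volume exotic explosions with pitched-up water splashes; medium mixes distant launcher explosions with the same splashes at half volume, pitched down.
# Concurrency override caps it at two instances, stopping the oldest.
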SoundCue /TidalCraneExotics/Gameplay/ShockwaveLauncher/Audio/SC_Exotic_Launcher_ShocknSlow_Projectile_Explode_Water
root=Mixer_2
nodes [
 {
  id=Attenuation_0
  children [Modulator_13]
  properties {attenuationSettings=/Game/Sounds/Attenuation/Fort_Explosion_Medium_Athena_Grenade_Att}
 }
 {id=Attenuation_1, children [Mixer_0], properties {attenuationSettings=/Game/Sounds/Attenuation/Fort_Explosion_Close_Athena_Grenade_Att}}
 {id=Mixer_0, children [Random_0,Modulator_0], properties {inputVolume [0.5,1]}}
 {id=Mixer_1, children [Random_5,Random_1], properties {inputVolume [1,0.5]}}
 {id=Mixer_2, children [Attenuation_1, Attenuation_0], properties {inputVolume [1,1]}}
 {id=Modulator_0, children [Random_2], properties {pitchMax=1.5, pitchMin=1.4}}
 {id=Modulator_13, children [Mixer_1], properties {pitchMax=0.95, pitchMin=0.85, volumeMax=1.1, volumeMin=1}}
 {id=Random_0, children [WavePlayer_0,WavePlayer_1,WavePlayer_2], properties {weights [1,1,1]}}
 {id=Random_1, children [WavePlayer_7,WavePlayer_8,WavePlayer_9], properties {weights [1,1,1]}}
 {id=Random_2, children [WavePlayer_3,WavePlayer_4,WavePlayer_5], properties {weights [1,1,1]}}
 {id=Random_5, children [WavePlayer_15, WavePlayer_16, WavePlayer_17], properties {weights [1,1,1]}}
 {
  id=WavePlayer_0
  properties {soundWaveAssetPtr=/TidalCraneExotics/Gameplay/ShockwaveLauncher/Audio/SoundWavs/Exotic_Shockwave_ShocknSlow_Explode_01}
 }
 {
  id=WavePlayer_1
  properties {soundWaveAssetPtr=/TidalCraneExotics/Gameplay/ShockwaveLauncher/Audio/SoundWavs/Exotic_Shockwave_ShocknSlow_Explode_02}
 }
 {
  id=WavePlayer_15
  properties {soundWaveAssetPtr=/TidalCraneExotics/Gameplay/ShockwaveLauncher/Audio/SoundWavs/Shockwave_Launcher_Explode_Distant_01}
 }
 {
  id=WavePlayer_16
  properties {soundWaveAssetPtr=/TidalCraneExotics/Gameplay/ShockwaveLauncher/Audio/SoundWavs/Shockwave_Launcher_Explode_Distant_02}
 }
 {
  id=WavePlayer_17
  properties {soundWaveAssetPtr=/TidalCraneExotics/Gameplay/ShockwaveLauncher/Audio/SoundWavs/Shockwave_Launcher_Explode_Distant_03}
 }
 {
  id=WavePlayer_2
  properties {soundWaveAssetPtr=/TidalCraneExotics/Gameplay/ShockwaveLauncher/Audio/SoundWavs/Exotic_Shockwave_ShocknSlow_Explode_03}
 }
 {
  id=WavePlayer_3
  properties {soundWaveAssetPtr=/TidalCraneExotics/Gameplay/ShockwaveLauncher/Audio/SoundWavs/Chrome_Splash_Grenade_Explo_Close_Water_01}
 }
 {
  id=WavePlayer_4
  properties {soundWaveAssetPtr=/TidalCraneExotics/Gameplay/ShockwaveLauncher/Audio/SoundWavs/Chrome_Splash_Grenade_Explo_Close_Water_02}
 }
 {
  id=WavePlayer_5
  properties {soundWaveAssetPtr=/TidalCraneExotics/Gameplay/ShockwaveLauncher/Audio/SoundWavs/Chrome_Splash_Grenade_Explo_Close_Water_03}
 }
 {
  id=WavePlayer_7
  properties {soundWaveAssetPtr=&Chrome_Splash_Grenade_Explo_Close_Water_01}
 }
 {
  id=WavePlayer_8
  properties {soundWaveAssetPtr=&Chrome_Splash_Grenade_Explo_Close_Water_02}
 }
 {
  id=WavePlayer_9
  properties {soundWaveAssetPtr=&Chrome_Splash_Grenade_Explo_Close_Water_03}
 }
]
properties {
 bOverrideConcurrency=true
 maxDistance=9625
 soundClassObject=/Game/Sounds/FortSoundClassesAndModes/Fort_Weapons
 volumeMultiplier=0.6
 attenuationOverrides {distanceAlgorithm=NaturalSound, falloffDistance=3800, lPFRadiusMax=3000, lPFRadiusMin=2000, attenuationShapeExtents {x=2200}}
 concurrencyOverrides {maxCount=2, resolutionRule=StopOldest}
}
