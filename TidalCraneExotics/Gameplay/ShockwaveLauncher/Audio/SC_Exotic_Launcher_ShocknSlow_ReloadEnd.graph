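# Sound cue for the end of the reload on the ShocknSlow exotic Shockwave Launcher, played as quiet weapon foley with reload attenuation.
# A mixer layers a randomly picked, modulated equip sound over a pitched-down stack of three reload insert sounds, which play together rather than as random variants.
# The cue raises overall pitch slightly and overrides concurrency to one instance, stopping the farthest.
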SoundCue /TidalCraneExotics/Gameplay/ShockwaveLauncher/Audio/SC_Exotic_Launcher_ShocknSlow_ReloadEnd
root=Mixer_1
nodes [
 {id=Mixer_1, children [Modulator_1,Modulator_0], properties {inputVolume [1,1]}}
 {id=Mixer_3, children [WavePlayer_0,WavePlayer_2,WavePlayer_7], properties {inputVolume [1,1,1]}}
 {id=Modulator_0, children [Mixer_3], properties {pitchMax=0.8, pitchMin=0.8}}
 {id=Modulator_1, children [Random_1]}
 {id=Random_1, children [WavePlayer_8,WavePlayer_9], properties {weights [1,1]}}
 {
  id=WavePlayer_0
  properties {soundWaveAssetPtr=/TidalCraneExotics/Gameplay/ShockwaveLauncher/Audio/SoundWavs/Shockwave_Launcher_ReloadInsert_01}
 }
 {
  id=WavePlayer_2
  properties {soundWaveAssetPtr=/TidalCraneExotics/Gameplay/ShockwaveLauncher/Audio/SoundWavs/Shockwave_Launcher_ReloadInsert_02}
 }
 {
  id=WavePlayer_7
  properties {soundWaveAssetPtr=/TidalCraneExotics/Gameplay/ShockwaveLauncher/Audio/SoundWavs/Shockwave_Launcher_ReloadInsert_03}
 }
 {id=WavePlayer_8, properties {soundWaveAssetPtr=/TidalCraneExotics/Gameplay/ShockwaveLauncher/Audio/SoundWavs/Shockwave_Launcher_Equip_03}}
 {id=WavePlayer_9, properties {soundWaveAssetPtr=/TidalCraneExotics/Gameplay/ShockwaveLauncher/Audio/SoundWavs/Shockwave_Launcher_Equip_01}}
]
properties {
 attenuationSettings=&Fort_Weapon_Reload_Att
 bOverrideConcurrency=true
 maxDistance=2250
 pitchMultiplier=1.1
 soundClassObject=&Fort_WeaponFoley
 volumeMultiplier=0.25
 attenuationOverrides {falloffDistance=5000, attenuationShapeExtents {x=1000}}
 concurrencyOverrides {maxCount=1, resolutionRule=StopFarthestThenPreventNew}
}
