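# Sound cue for the reload spin foley of the exotic Shock 'n Slow shockwave launcher.
# It picks one of three equally weighted spin recordings, then runs it through a modulator for slight variation.
# The cue plays quiet and pitched up in the weapon foley sound class, using the shared weapon reload attenuation.
# Overridden concurrency allows one instance, stopping the farthest and then preventing new ones.
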SoundCue /TidalCraneExotics/Gameplay/ShockwaveLauncher/Audio/SC_Exotic_Launcher_ShocknSlow_ReloadSpin
root=Modulator_1
nodes [
 {id=Modulator_1, children [Random_1], properties {pitchMin=1}}
 {id=Random_1, children [WavePlayer_3,WavePlayer_4,WavePlayer_5], properties {weights [1,1,1]}}
 {
  id=WavePlayer_3
  properties {soundWaveAssetPtr=/TidalCraneExotics/Gameplay/ShockwaveLauncher/Audio/SoundWavs/Exotic_Shockwave_ShocknSlow_Reload_Spin_01}
 }
 {
  id=WavePlayer_4
  properties {soundWaveAssetPtr=/TidalCraneExotics/Gameplay/ShockwaveLauncher/Audio/SoundWavs/Exotic_Shockwave_ShocknSlow_Reload_Spin_02}
 }
 {
  id=WavePlayer_5
  properties {soundWaveAssetPtr=/TidalCraneExotics/Gameplay/ShockwaveLauncher/Audio/SoundWavs/Exotic_Shockwave_ShocknSlow_Reload_Spin_03}
 }
]
properties {
 attenuationSettings=&Fort_Weapon_Reload_Att
 bOverrideConcurrency=true
 maxDistance=2250
 pitchMultiplier=1.3
 soundClassObject=&Fort_WeaponFoley
 volumeMultiplier=0.3
 attenuationOverrides {falloffDistance=5000, attenuationShapeExtents {x=1000}}
 concurrencyOverrides {maxCount=1, resolutionRule=StopFarthestThenPreventNew}
}
