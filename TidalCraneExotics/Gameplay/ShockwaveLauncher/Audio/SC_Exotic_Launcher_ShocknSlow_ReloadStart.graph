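# Sound cue for the start of the reload on the Shock 'n Slow exotic shockwave launcher.
# A modulator feeds from a random node that picks one of three equally weighted reload-start wave variations.
# It plays quietly as weapon foley using the shared weapon reload attenuation, with a concurrency override that allows one instance and favors stopping the farthest.
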SoundCue /TidalCraneExotics/Gameplay/ShockwaveLauncher/Audio/SC_Exotic_Launcher_ShocknSlow_ReloadStart
root=Modulator_1
nodes [
 {id=Modulator_1, children [Random_3]}
 {id=Random_3, children [WavePlayer_9, WavePlayer_10, WavePlayer_11], properties {weights [1,1,1]}}
 {
  id=WavePlayer_10
  properties {soundWaveAssetPtr=/TidalCraneExotics/Gameplay/ShockwaveLauncher/Audio/SoundWavs/Shockwave_Launcher_ReloadStart_02}
 }
 {
  id=WavePlayer_11
  properties {soundWaveAssetPtr=/TidalCraneExotics/Gameplay/ShockwaveLauncher/Audio/SoundWavs/Shockwave_Launcher_ReloadStart_03}
 }
 {
  id=WavePlayer_9
  properties {soundWaveAssetPtr=/TidalCraneExotics/Gameplay/ShockwaveLauncher/Audio/SoundWavs/Shockwave_Launcher_ReloadStart_01}
 }
]
properties {
 attenuationSettings=&Fort_Weapon_Reload_Att
 bOverrideConcurrency=true
 maxDistance=2250
 soundClassObject=&Fort_WeaponFoley
 volumeMultiplier=0.2
 attenuationOverrides {falloffDistance=5000, attenuationShapeExtents {x=1000}}
 concurrencyOverrides {maxCount=1, resolutionRule=StopFarthestThenPreventNew}
}
