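; Gameplay ability that grants the Shockwave Launcher's hang time movement effect to its owning avatar.
; On activation it starts the registered MME_ShockGrenade_Hangtime movement mode extension on the avatar actor.
; When the ability ends, cancelled or not, it cancels the active extension by the MovementMode.HangTime.ShockGrenade tag, so the ability's lifetime bounds the effect.
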(event Ability|EventActivateAbility
  (Class|FortMovementModeExtBlueprintLibrary|ActivateRegisteredMovementModeExtension "/TidalCraneExotics/Gameplay/ShockwaveLauncher/MME_ShockGrenade_Hangtime.MME_ShockGrenade_Hangtime_C" (Ability|GetAvatarActorfromActorInfo)))

(event Ability|EventOnEndAbility
  (params (bWasCancelled :type Boolean))
  (Class|FortMovementModeExtBlueprintLibrary|CancelActiveExtensionwithTag (Ability|GetAvatarActorfromActorInfo) "(GameplayTags=((TagName=\"MovementMode.HangTime.ShockGrenade\")))"))
